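# Niagara system for the swing trail of the EnsureHall Snag pickaxe, drawn as a dark, slightly blue-tinted ribbon.
# Its single Trail emitter uses anim-trail modules to spawn and update particles from the system's skeletal mesh, with a random custom ID set per emitter.
# Initial color comes from a system TOD (time-of-day) value, then a dim color curve scales it.
# Alpha multiplies system-age and particle-age curves for a quick, faint fade.
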NiagaraSystem /BRCosmetics/Weapons/FORT_Melee/Pickaxe_EnsureHall_Snag/FX/NS_Pickaxe_EnsureHall_Snag_Trail
dependencies {
 usedDataInterfaces [&NiagaraDataInterfaceColorCurve, &NiagaraDataInterfaceCurve, &NiagaraDataInterfaceSkeletalMesh]
 usedDynamicInputs [&FNTimeOfDayScalar]
 usedModules [
  &self:SystemSpawnScript.SystemScriptSource.SystemScriptGraph.NiagaraNodeAssignment_0.SetVariables_#1
  &SystemState
  &self:SystemSpawnScript.SystemScriptSource.SystemScriptGraph.NiagaraNodeAssignment_1.SetVariables_#2
  &self:SystemSpawnScript.SystemScriptSource.SystemScriptGraph.NiagaraNodeAssignment_2.SetVariables_#3
 ]
}
emitters {Trail {
  emitterSpawnScript [
   &FN_RibbonIDOffsetUtility()
   /SetVariables(Emitter.CustomID:NiagaraID=/Niagara/DynamicInputs/NiagaraID/MakeNiagaraIDFromInt(Index:NiagaraInt32=&RandomRangeInt(Minimum=0, Maximum=1000), Acquire Tag:NiagaraInt32=0)) {isSetParameters=true}
  ]
  emitterUpdateScript [
   &EmitterState()
   &FN_SpawnAnimTrail(Spawn Interval=0, Switch on ENiagara_AnimTrail_Source:ENiagara_AnimTrail_Source=::NewEnumerator1)
  ]
  particleSpawnScript [
   &FN_InitializeAnimTrailBoneIndecies()
   &FN_InitializeAnimTrail(
    Color=/Niagara/DynamicInputs/TypeConversions/MakeLinearColorFromFloat(Value:NiagaraFloat=$System.TOD)
    Lifetime=0.8
    SkeletalMesh=$System.Skeletal Mesh
   )
   /SetVariables(Particles.CustomParticleID:NiagaraID=$Emitter.CustomID) {isSetParameters=true}
  ]
  particleUpdateScript [
   &SolveForcesAndVelocity(Clamp Velocity=false, Limit Acceleration=false)
   &FN_UpdateAnimTrail(SkeletalMesh=$System.Skeletal Mesh) {module=FN_UpdateAnimTrail001}
   &ScaleColor(
    Scale Alpha=&Multiply_Float(
     A=&FloatFromCurve(
      FloatCurve=DataInterface() {propertyValues {Curve {keys [
          {
           arriveTangent=0.0468
           interpMode=RCIM_Cubic
           leaveTangent=0.0468
           leaveTangentWeight=0.3333
           tangentMode=RCTM_User
           time=0.0024
           value=0.9617
          }
          {arriveTangent=-12.8084, interpMode=RCIM_Cubic, leaveTangent=-12.8084, tangentMode=RCTM_User, time=0.1048, value=0.4421}
          {
           arriveTangent=-0.0274
           arriveTangentWeight=0.0175
           interpMode=RCIM_Cubic
           leaveTangent=0.0323
           leaveTangentWeight=0.2667
           tangentMode=RCTM_Break
           time=0.1656
           value=0.0018
          }
          <at 1 0>
         ]}}}
      CurveIndex=$Engine.System.Age
      Scale Curve=0.2
     )
     B=&FloatFromCurve(
      FloatCurve=DataInterface() {propertyValues {Curve {keys [
          {interpMode=RCIM_Cubic, tangentMode=RCTM_User, time=0, value=1}
          {arriveTangent=-11.9871, interpMode=RCIM_Cubic, leaveTangent=-11.9871, tangentMode=RCTM_User, time=0.1688, value=0.4093}
          {arriveTangent=-2.605, interpMode=RCIM_Cubic, leaveTangent=-2.605, tangentMode=RCTM_User, time=0.2, value=0}
         ]}}}
      CurveIndex=$Particles.Age
     )
    )
    Scale RGB=&MakeVectorFromLinearColorRGB(Color=&ColorFromCurve(ColorCurve=DataInterface() {propertyValues {
        AlphaCurve {keys [<at 0.3289 1 RCIM_Cubic>,<at 1 0 RCIM_Cubic>]}
        BlueCurve {keys [<at 0 0.06>,<at 1 0.0174>]}
        GreenCurve {keys [<at 0 0.048>,<at 1 0.011>]}
        RedCurve {keys [<at 0 0.048>,<at 1 0.011>]}
       }}))
    ScaleA=true
    ScaleRGB=true
   )
  ]
  rendererClasses [&NiagaraRibbonRendererProperties]
  renderers [{rendererClass=&NiagaraRibbonRendererProperties}]
 }}
summary {
 systemName=NS_Pickaxe_EnsureHall_Snag_Trail
 emitters [{emitterName=Trail, rendererClasses [&NiagaraRibbonRendererProperties]}]
 userVariables [
  User.bIsFrontEnd {type=NiagaraBool}
  User.bIsFrontEndPreview {type=NiagaraBool}
  User.RibbonIDOffset {type=NiagaraInt32}
 ]
}
